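# Translucent unlit Niagara particle material for the Veiled emote's glitch swap decal effect.
# Scene depth reconstructs particle-local position behind the surface, driving a box mask and distorted gradient wipe for a projected-decal look.
# Noise-warped curve-atlas lines, stepped screen-space blocks, scanlines and a colored grid build the glitch layers, passed between sections by named reroutes.
# Opacity blends two masked variants by UseAlpha, then applies camera depth fade.
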Material /Game/Animation/Game/MainPlayer/Emotes/Veiled/FX/Materials/M_Glitch_Swap_Decal_Veiled_Emote
graph {
 properties {
  bUsedWithNiagaraMeshParticles=true
  bUsedWithNiagaraSprites=true
  blendMode=BLEND_Translucent
  pixelDepthOffsetMode=PDOM_Legacy
  shadingModel=MSM_Unlit
  usedShadingModels=Unlit
 }
 nodes [
  MaterialFunctionCall(Emissive=Add(A=NamedRerouteUsage(), B=NamedRerouteUsage())) {materialFunction=&MF_FX_EmissiveAdjustments}
  n1 = NamedRerouteUsage()
  n2 = Multiply(A=NamedRerouteUsage(), B=@n1) {constB=5}
  n3 = NamedRerouteUsage()
  n4 = Multiply(A=@n1, B=TextureSample() {texture=/Game/Animation/Game/MainPlayer/Emotes/Veiled/FX/Textures/T_Veiled_Merged_Pattern}.A)
  n5 = NamedRerouteUsage()
  n6 = ScalarParameter() {defaultValue=220, parameterName=FadeLength}
  n7 = ScalarParameter() {defaultValue=80, parameterName=FadeOffset}
  Multiply(
   A=LinearInterpolate(A=Multiply(A=@n2, B=@n3), B=Multiply(A=@n4, B=@n5), Alpha=ScalarParameter() {parameterName=UseAlpha})
   B=Saturate(None=MaterialFunctionCall(Fade Length=@n6, Fade Offset=@n7) {materialFunction=&CameraDepthFade}.Result)
  )
  n8 = TextureSample(UVs=Panner() {desc="-0.33,-0.43", speedX=-0.1, speedY=0.1}) {texture=/Game/Characters/Player/Female/Medium/Bodies/F_MED_Veiled_Glitch/FX/T_Veiled_Glitch_Noise_Marbling}
  n9 = ScalarParameter() {defaultValue=1, group=Dance Lines, parameterName=StretchGradient}
  n10 = Time()
  n11 = ScalarParameter() {defaultValue=1, group=Dance Lines, parameterName=speed}
  n12 = Add(
   A=Multiply(A=ComponentMask(None=@n8.RGB) {r=true}, B=ScalarParameter() {defaultValue=1, group=Dance Lines, parameterName=NoiseWarpIntensity})
   B=Add(A=Multiply(B=@n9) {constA=1}, B=Multiply(A=@n10, B=@n11))
  )
  n13 = ScalarParameter() {defaultValue=0.75, group=Dance Lines, parameterName=Repetition}
  n14 = CurveAtlasRowParameter(CurveTime=Frac(None=Multiply(A=@n12, B=@n13))) {
   atlas=/Game/Characters/Player/Female/Medium/Bodies/F_MED_Veiled_Glitch/FX/CA_Veiled_Glitch
   curve=/Game/Characters/Player/Female/Medium/Bodies/F_MED_Veiled_Glitch/FX/CC_Veiled_Glitch
  }
  NamedRerouteDeclaration(None=Multiply(A=Normalize(VectorInput=@n14), B=ParticleColor().RGB)) {name=Veiled_Pattern, nodeColor=<rgba 1 0.3452 0 1>}
  n15 = Multiply(A=NamedRerouteUsage()) {constB=0.2}
  n16 = Constant3Vector() {constant=<rgba 0.1606 0.1954 0.4896 1>}
  n17 = Constant3Vector() {constant=<rgba 0.0303 0.0595 0.3385 0>}
  n18 = ScreenPosition()
  n19 = Constant2Vector() {g=5, r=2.5}
  n20 = Multiply(A=Time(), B=ScalarParameter() {defaultValue=10, group=Block, parameterName=Block U Speed}) {constB=0.5}
  n21 = TextureSample(UVs=Panner(Coordinate=Multiply(A=@n18.ViewportUV, B=@n19), Time=Ceil(None=@n20)) {desc="-0.33,-0.43", speedX=-0.3, speedY=0.5}) {texture=&T_Veiled_Merged_Pattern}
  n22 = LinearInterpolate(Alpha=@n21.R) {constA=-0.2, constB=2}
  n23 = Constant() {r=1}
  n24 = Clamp(None=NamedRerouteDeclaration(None=StaticSwitchParameter(True=@n22, False=@n23) {defaultValue=true, parameterName=bUseBlocks}) {name=Blocks, nodeColor=<rgba 1 0 0.1357 1>})
  n25 = Constant3Vector() {constant=<rgba 0.1198 0.1982 1 1>}
  n26 = Constant3Vector() {constant=<rgba 1 0 0.0174 0>}
  n27 = ComponentMask(None=TextureCoordinate()) {g=true}
  n28 = Desaturation(None=Add(A=LinearInterpolate(A=@n16, B=@n17, Alpha=@n24), B=LinearInterpolate(A=@n25, B=@n26, Alpha=@n27)), Fraction=Constant() {r=-2.5}) {luminanceFactors=<rgba 0.59 0 0.1923 0>}
  n29 = Add(A=@n15, B=@n28)
  n30 = Multiply(A=Multiply(A=@n15, B=@n28)) {constB=-2.5}
  NamedRerouteDeclaration(None=Multiply(A=Add(A=@n29, B=@n30), B=TextureSample() {texture=&T_Veiled_Merged_Pattern}.G)) {name=BlockEmi, nodeColor=<rgba 1 0 0.1357 1>}
  n31 = Divide(A=SceneDepth().R, B=PixelDepth().R)
  n32 = CameraPositionWS()
  n33 = Subtract(A=WorldPosition().XYZ, B=@n32)
  n34 = TransformPosition(None=Add(A=Multiply(A=@n31, B=@n33), B=@n32)) {transformSourceType=TRANSFORMPOSSOURCE_World, transformType=TRANSFORMPOSSOURCE_Particle}
  n35 = MaterialFunctionCall!(
   A=@n34
   B
   Bounds=VectorParameter() {group=Masking, parameterName=BoxMaskBounds, defaultValue=<rgba 1 1 1 1>}.RGB
   Edge Falloff=ScalarParameter() {group=Masking, parameterName=BoxMask_Falloff}
  ) {materialFunction=/Engine/Functions/Engine_MaterialFunctions01/Gradient/BoxMask-3D}
  n36 = NamedRerouteUsage()
  n37 = Add(A=Multiply() {constB=2}) {constB=-1}
  n38 = ScalarParameter() {defaultValue=0.05, parameterName=Blocks_DistortionStr}
  n39 = ComponentMask(None=NamedRerouteUsage()) {b=true, r=true}
  n40 = Constant2Vector() {g=2, r=0.5}
  n41 = TextureSample(UVs=Panner(Coordinate=Multiply(A=@n39, B=@n40) {constB=3}) {speedX=1}) {texture=&T_Inky_Smoke_Tile}
  n42 = Multiply(A=Subtract(A=Multiply(A=@n41.G) {constB=2})) {constB=0.2}
  n43 = Constant()
  n44 = Add(
   A=Add(A=ComponentMask(None=@n36) {r=true}, B=Multiply(A=@n37, B=@n38))
   B=QualitySwitch(Default=@n42, Low=@n43, High=@n42, Medium=@n43, Epic=@n42)
  )
  NamedRerouteDeclaration(None=Multiply(
    A=NamedRerouteDeclaration(None=@n35.Result) {name=VolumeMask, nodeColor=<rgba 1.7485 0 3 1>}
    B=TextureSample(UVs=@n44) {samplerSource=SSM_Clamp_WorldGroupSettings, texture=/Game/Gadgets/Assets/Misty_Bop/Effects/T_MediumGradient_Horizontal}.R
   )) {name=WipeMask, nodeColor=<rgba 1.7485 0 3 1>}
  n45 = OneMinus(None=ParticleColor().A)
  n46 = ScalarParameter() {parameterName=Alpha_W}
  NamedRerouteDeclaration(None=SmoothStep(Min=@n45, Max=Add(A=@n45, B=@n46), Value=@n21.R)) {name=BlocksAlphaRamp, nodeColor=<rgba 1 0 0.1357 1>}
  n47 = ScalarParameter() {defaultValue=8, group=Block, parameterName=Line Dark}
  n48 = ScalarParameter() {defaultValue=12, group=Block, parameterName=Line Light}
  n49 = MaterialFunctionCall(Float2=ScreenPosition().ViewportUV) {materialFunction=&BreakOutFloat2Components}
  n50 = ScalarParameter() {defaultValue=64, group=Block, parameterName=Line amount}
  n51 = Time()
  n52 = ScalarParameter() {defaultValue=0.2, group=Block, parameterName=Line Speed}
  n53 = Frac(None=Add(A=Multiply(A=@n49.G, B=@n50), B=Multiply(A=@n51, B=@n52)))
  n54 = OneMinus(None=@n53)
  n55 = Ceil(None=Add(A=Multiply(A=@n53, B=@n54), B=ScalarParameter() {defaultValue=-0.2, group=Block, parameterName=Line Size}))
  NamedRerouteDeclaration(None=Multiply(A=LinearInterpolate(A=@n47, B=@n48, Alpha=@n55), B=Constant() {r=0.01})) {name=Glitch_Lines, nodeColor=<rgba 0 0.5983 1 1>}
  n56 = Panner() {speedY=1}
  n57 = Time()
  n58 = Add(A=Sine(None=Multiply(A=@n57) {constB=0.5})) {constB=1.5}
  n59 = TextureCoordinate()
  n60 = TextureCoordinate()
  n61 = Constant2Vector() {g=4, r=2}
  n62 = Multiply(
   A=Power(Base=ComponentMask(None=@n59) {g=true}) {constExponent=1.5}
   B=TextureSample(UVs=Multiply(A=@n60, B=@n61) {constB=4}) {texture=&T_Veiled_Merged_Pattern}.B
  )
  n63 = TextureCoordinate()
  n64 = Multiply(
   A=Saturate(None=Multiply(A=@n58, B=@n62))
   B=LinearInterpolate(
    A=Constant3Vector() {constant=<rgba 0 0.2804 1 1>}
    B=Constant3Vector() {constant=<rgba 1 0 0.2228 0>}
    Alpha=ComponentMask(None=@n63) {g=true}
   )
  )
  NamedRerouteDeclaration(None=Multiply(
    A=TextureSample(UVs=@n56) {texture=&T_Veiled_Merged_Pattern}.G
    B=Saturate(None=@n64)
   )) {name=Grid, nodeColor=<rgba 0 1 0.3919 1>}
  n65 = ComponentMask(None=@n34) {b=true, g=true, r=true}
  NamedRerouteDeclaration(None=Saturate(None=Add(A=@n65) {constB=0.5})) {name=MeshParticle_LocalPositionBehindTransparency, nodeColor=<rgba 1.7485 0 3 1>}
 ]
}
